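# Physics asset for the ponytail hair of the female medium Impulse Spring skin ASN_01.
# A base box on the head leads into two stiff root capsules, then splits at dyn_ponyTail_02 into three capsule chains.
# Swing limits widen progressively toward the tips while twist stays tight along each chain.
# Upper-body capsules on spine, neck, head and arms, plus large clavicle boxes, give body shapes linked by angularly free constraints.
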PhysicsAsset /Game/Characters/Player/Female/Medium/Bodies/F_MED_Impulse_Spring/Skins/ASN_01/Meshes/F_MED_Impulse_Spring_ASN_1_FaceAcc_hair_Physics
bodies {
 clavicle_l [
  {shapeType=Capsule, length=15.434, radius=7.521, center=<-0.05 -0.6853 -1.245>, rotation=<rot 89.7019 -0.0003 -0.0003>}
  {shapeType=Box, extentX=33.6329, extentY=25.2102, extentZ=37.5919, center=<6.3561 13.6644 22.9098>, rotation=<rot -10.0005 0 0>}
 ]
 clavicle_r [
  {shapeType=Capsule, length=16.2726, radius=6.6824, center=<-0.1324 0.6199 -1.2505>, rotation=<rot 89.7019 -0.0003 -0.0003>}
  {shapeType=Box, extentX=29.6329, extentY=22.9787, extentZ=45.5919, center=<-5.7804 -11.3756 -21.1597>, rotation=<rot -3.8055 0 0>}
 ]
 dyn_ponyTail_01 [{
   shapeType=Capsule
   length=2.9414
   radius=2.5176
   shapeName=dyn_ponyTail_01_capsule
   center=<1.678 -0.0341 0.7063>
   rotation=<rot 0 174.007 -59.784>
  }]
 dyn_ponyTail_02 [{
   shapeType=Capsule
   length=4.9357
   radius=3.9472
   shapeName=dyn_ponyTail_02_capsule
   center=<1.4818 -0.1283 0.2461>
   rotation=<rot 0 177.1544 -87.7221>
  }]
 dyn_ponyTail_03 [{
   shapeType=Capsule
   length=3.5078
   radius=2.7519
   shapeName=dyn_ponyTail_03_capsule
   center=<2.9892 0.0121 -0.0755>
   rotation=<rot 11.3969 -90 -72.4963>
  }]
 dyn_ponyTail_04 [{shapeType=Capsule, length=4.2035, radius=1.9389, shapeName=dyn_ponyTail_04_capsule, center=<2.3882 0 0>, rotation=<rot 0 90 90>}]
 dyn_ponyTail_05 [{shapeType=Capsule, length=4.4157, radius=2.0466, shapeName=dyn_ponyTail_05_capsule, center=<2.2545 0 0>, rotation=<rot 0 90 90>}]
 dyn_ponyTail_06 [{
   shapeType=Capsule
   length=4.1616
   radius=1.7422
   shapeName=dyn_ponyTail_06_capsule
   center=<5.0874 0.1846 -0.7227>
   rotation=<rot 0 -94.5318 -80.6785>
  }]
 dyn_ponyTail_07 [{shapeType=Capsule, length=2, radius=2, shapeName=dyn_ponyTail_07_capsule, center=<2.4449 0 0>, rotation=<rot 90 0 0>}]
 dyn_ponyTail_08 [{shapeType=Capsule, length=2, radius=2, shapeName=dyn_ponyTail_08_capsule, center=<2.4449 0 0>, rotation=<rot 90 0 0>}]
 dyn_ponyTail_09 [{shapeType=Capsule, length=2, radius=2, shapeName=dyn_ponyTail_09_capsule, center=<2.4449 0 0>, rotation=<rot 90 0 0>}]
 dyn_ponyTail_10 [{shapeType=Capsule, length=2, radius=2, shapeName=dyn_ponyTail_10_capsule, center=<2.4449 0 0>, rotation=<rot 90 0 0>}]
 dyn_ponyTail_11 [{shapeType=Capsule, length=2, radius=2, shapeName=dyn_ponyTail_11_capsule, center=<2.4449 0 0>, rotation=<rot 90 0 0>}]
 dyn_ponyTail_12 [{shapeType=Capsule, length=3.6419, radius=2.3303, shapeName=dyn_ponyTail_12_capsule, center=<2.7089 0 0>, rotation=<rot 0 90 -90>}]
 dyn_ponyTail_13 [{shapeType=Capsule, length=4.0123, radius=1.822, shapeName=dyn_ponyTail_13_capsule, center=<2.3882 0 0>, rotation=<rot 0 90 90>}]
 dyn_ponyTail_14 [{shapeType=Capsule, length=4.3917, radius=1.9937, shapeName=dyn_ponyTail_14_capsule, center=<2.3882 0 0>, rotation=<rot 0 90 90>}]
 dyn_ponyTail_15 [{
   shapeType=Capsule
   length=4.0455
   radius=1.673
   shapeName=dyn_ponyTail_15_capsule
   center=<4.9829 0.1313 -0.5019>
   rotation=<rot 0 -89.0595 -82.8308>
  }]
 dyn_ponyTail_16 [{shapeType=Capsule, length=2, radius=2, shapeName=dyn_ponyTail_16_capsule, center=<2.4449 0 0>, rotation=<rot 90 0 0>}]
 dyn_ponyTail_17 [{shapeType=Capsule, length=2, radius=2, shapeName=dyn_ponyTail_17_capsule, center=<2.4449 0 0>, rotation=<rot 90 0 0>}]
 dyn_ponyTail_18 [{shapeType=Capsule, length=2, radius=2, shapeName=dyn_ponyTail_18_capsule, center=<2.4449 0 0>, rotation=<rot 90 0 0>}]
 dyn_ponyTail_19 [{shapeType=Capsule, length=2, radius=2, shapeName=dyn_ponyTail_19_capsule, center=<2.4449 0 0>, rotation=<rot 90 0 0>}]
 dyn_ponyTail_20 [{shapeType=Capsule, length=2, radius=2, shapeName=dyn_ponyTail_20_capsule, center=<2.4449 0 0>, rotation=<rot 90 0 0>}]
 dyn_ponyTail_21 [{shapeType=Capsule, length=3.3497, radius=2.1063, shapeName=dyn_ponyTail_21_capsule, center=<2.4243 0 0>, rotation=<rot 0 90 90>}]
 dyn_ponyTail_22 [{shapeType=Capsule, length=3.7738, radius=2.096, shapeName=dyn_ponyTail_22_capsule, center=<2.2545 0 0>, rotation=<rot 0 90 90>}]
 dyn_ponyTail_23 [{shapeType=Capsule, length=4.4669, radius=1.9693, shapeName=dyn_ponyTail_23_capsule, center=<2.2545 0 0>, rotation=<rot 0 90 90>}]
 dyn_ponyTail_24 [{
   shapeType=Capsule
   length=3.9984
   radius=1.4819
   shapeName=dyn_ponyTail_24_capsule
   center=<4.9347 -0.0545 -0.3058>
   rotation=<rot 0 -85.9732 -89.1142>
  }]
 dyn_ponyTail_25 [{shapeType=Capsule, length=2, radius=2, shapeName=dyn_ponyTail_25_capsule, center=<2.4449 0 0>, rotation=<rot 90 0 0>}]
 dyn_ponyTail_26 [{shapeType=Capsule, length=2, radius=2, shapeName=dyn_ponyTail_26_capsule, center=<2.4449 0 0>, rotation=<rot 90 0 0>}]
 dyn_ponyTail_27 [{shapeType=Capsule, length=2, radius=2, shapeName=dyn_ponyTail_27_capsule, center=<2.4449 0 0>, rotation=<rot 90 0 0>}]
 dyn_ponyTail_28 [{shapeType=Capsule, length=2, radius=2, shapeName=dyn_ponyTail_28_capsule, center=<2.4449 0 0>, rotation=<rot 90 0 0>}]
 dyn_ponyTail_29 [{shapeType=Capsule, length=2, radius=2, shapeName=dyn_ponyTail_29_capsule, center=<2.4449 0 0>, rotation=<rot 90 0 0>}]
 dyn_ponyTail_Base [{shapeType=Box, extentX=0.5606, extentY=6.325, extentZ=8.32, center=<0.7102 0 0>}]
 hand_l [{shapeType=Capsule, length=5, radius=5, rotation=<rot 90 0 0>}]
 hand_r [{shapeType=Capsule, length=5, radius=5, rotation=<rot 90 0 0>}]
 head [{shapeType=Capsule, length=4.1295, radius=7.2822, center=<5.5315 -0.8189 0>, rotation=<rot 90 0 0>}]
 lowerarm_l [{shapeType=Capsule, length=16.7317, radius=6.5514, center=<12.0412 0 0>, rotation=<rot 90 0 0>}]
 lowerarm_r [{shapeType=Capsule, length=15.9692, radius=7.8848, center=<-11.7422 0 0>, rotation=<rot 90 0 0>}]
 lowerarm_twist_02_r [{shapeType=Capsule, length=5, radius=5, rotation=<rot 90 0 0>}]
 neck_01 [{shapeType=Capsule, length=5.25, radius=11.0429}]
 neck_02 [{shapeType=Capsule, length=3.7447, radius=5.1833}]
 spine_01 [{shapeType=Capsule, length=8.58, radius=7.42}]
 spine_02 [{shapeType=Capsule, length=6.4891, radius=5.6158}]
 spine_03 [{shapeType=Capsule, length=10.4688, radius=5.5312}]
 spine_04 [{shapeType=Capsule, length=7.9301, radius=8.0699}]
 spine_05 [{shapeType=Capsule, length=2.9985, radius=12.5584, center=<7.0234 -0.2562 0>}]
 upperarm_l [{shapeType=Capsule, length=16.108, radius=6.1274, center=<13.3322 0.0002 0>, rotation=<rot 90 0 0>}]
 upperarm_r [{shapeType=Capsule, length=15.0659, radius=6.1904, center=<-12.1402 0 0>, rotation=<rot 90 0 0>}]
 upperarm_twist_01_l [{shapeType=Capsule, length=5, radius=5, rotation=<rot 90 0 0>}]
 upperarm_twist_01_r [{shapeType=Capsule, length=5, radius=5, rotation=<rot 90 0 0>}]
 upperarm_twist_02_l [{shapeType=Capsule, length=5, radius=5, rotation=<rot 90 0 0>}]
 upperarm_twist_02_r [{shapeType=Capsule, length=5, radius=5, rotation=<rot 90 0 0>}]
}
constraints [
 {bone1Name=dyn_ponyTail_01, bone2Name=dyn_ponyTail_Base, swing1LimitDegrees=5, swing2LimitDegrees=5, twistLimitDegrees=5}
 {bone1Name=dyn_ponyTail_02, bone2Name=dyn_ponyTail_01, swing1LimitDegrees=5, swing2LimitDegrees=5, twistLimitDegrees=5}
 {bone1Name=dyn_ponyTail_Base, bone2Name=head, swing1LimitDegrees=45, swing2LimitDegrees=45, twistLimitDegrees=45}
 {
  bone1Name=head
  bone2Name=neck_02
  swing1LimitDegrees=45
  swing1Motion=Free
  swing2LimitDegrees=45
  swing2Motion=Free
  twistLimitDegrees=45
  twistMotion=Free
 }
 {
  bone1Name=neck_02
  bone2Name=neck_01
  swing1LimitDegrees=45
  swing1Motion=Free
  swing2LimitDegrees=45
  swing2Motion=Free
  twistLimitDegrees=45
  twistMotion=Free
 }
 {
  bone1Name=neck_01
  bone2Name=spine_05
  swing1LimitDegrees=45
  swing1Motion=Free
  swing2LimitDegrees=45
  swing2Motion=Free
  twistLimitDegrees=45
  twistMotion=Free
 }
 {
  bone1Name=spine_05
  bone2Name=spine_04
  swing1LimitDegrees=45
  swing1Motion=Free
  swing2LimitDegrees=45
  swing2Motion=Free
  twistLimitDegrees=45
  twistMotion=Free
 }
 {
  bone1Name=spine_04
  bone2Name=spine_03
  swing1LimitDegrees=45
  swing1Motion=Free
  swing2LimitDegrees=45
  swing2Motion=Free
  twistLimitDegrees=45
  twistMotion=Free
 }
 {
  bone1Name=spine_03
  bone2Name=spine_02
  swing1LimitDegrees=45
  swing1Motion=Free
  swing2LimitDegrees=45
  swing2Motion=Free
  twistLimitDegrees=45
  twistMotion=Free
 }
 {
  bone1Name=spine_02
  bone2Name=spine_01
  swing1LimitDegrees=45
  swing1Motion=Free
  swing2LimitDegrees=45
  swing2Motion=Free
  twistLimitDegrees=45
  twistMotion=Free
 }
 {
  bone1Name=clavicle_l
  bone2Name=spine_05
  swing1LimitDegrees=45
  swing1Motion=Free
  swing2LimitDegrees=45
  swing2Motion=Free
  twistLimitDegrees=45
  twistMotion=Free
 }
 {
  bone1Name=upperarm_l
  bone2Name=clavicle_l
  swing1LimitDegrees=45
  swing1Motion=Free
  swing2LimitDegrees=45
  swing2Motion=Free
  twistLimitDegrees=45
  twistMotion=Free
 }
 {
  bone1Name=lowerarm_l
  bone2Name=upperarm_l
  swing1LimitDegrees=45
  swing1Motion=Free
  swing2LimitDegrees=45
  swing2Motion=Free
  twistLimitDegrees=45
  twistMotion=Free
 }
 {
  bone1Name=hand_l
  bone2Name=lowerarm_l
  swing1LimitDegrees=45
  swing1Motion=Free
  swing2LimitDegrees=45
  swing2Motion=Free
  twistLimitDegrees=45
  twistMotion=Free
 }
 {
  bone1Name=upperarm_twist_01_l
  bone2Name=upperarm_l
  swing1LimitDegrees=45
  swing1Motion=Free
  swing2LimitDegrees=45
  swing2Motion=Free
  twistLimitDegrees=45
  twistMotion=Free
 }
 {
  bone1Name=upperarm_twist_02_l
  bone2Name=upperarm_l
  swing1LimitDegrees=45
  swing1Motion=Free
  swing2LimitDegrees=45
  swing2Motion=Free
  twistLimitDegrees=45
  twistMotion=Free
 }
 {
  bone1Name=upperarm_r
  bone2Name=clavicle_r
  swing1LimitDegrees=45
  swing1Motion=Free
  swing2LimitDegrees=45
  swing2Motion=Free
  twistLimitDegrees=45
  twistMotion=Free
 }
 {
  bone1Name=lowerarm_r
  bone2Name=upperarm_r
  swing1LimitDegrees=45
  swing1Motion=Free
  swing2LimitDegrees=45
  swing2Motion=Free
  twistLimitDegrees=45
  twistMotion=Free
 }
 {
  bone1Name=hand_r
  bone2Name=lowerarm_r
  swing1LimitDegrees=45
  swing1Motion=Free
  swing2LimitDegrees=45
  swing2Motion=Free
  twistLimitDegrees=45
  twistMotion=Free
 }
 {
  bone1Name=lowerarm_twist_02_r
  bone2Name=lowerarm_r
  swing1LimitDegrees=45
  swing1Motion=Free
  swing2LimitDegrees=45
  swing2Motion=Free
  twistLimitDegrees=45
  twistMotion=Free
 }
 {
  bone1Name=upperarm_twist_01_r
  bone2Name=upperarm_r
  swing1LimitDegrees=45
  swing1Motion=Free
  swing2LimitDegrees=45
  swing2Motion=Free
  twistLimitDegrees=45
  twistMotion=Free
 }
 {
  bone1Name=upperarm_twist_02_r
  bone2Name=upperarm_r
  swing1LimitDegrees=45
  swing1Motion=Free
  swing2LimitDegrees=45
  swing2Motion=Free
  twistLimitDegrees=45
  twistMotion=Free
 }
 {
  bone1Name=clavicle_r
  bone2Name=spine_05
  swing1LimitDegrees=45
  swing1Motion=Free
  swing2LimitDegrees=45
  swing2Motion=Free
  twistLimitDegrees=45
  twistMotion=Free
 }
 {bone1Name=dyn_ponyTail_23, bone2Name=dyn_ponyTail_22, swing1LimitDegrees=20, swing2LimitDegrees=20, twistLimitDegrees=5}
 {bone1Name=dyn_ponyTail_24, bone2Name=dyn_ponyTail_23, swing1LimitDegrees=25, swing2LimitDegrees=25, twistLimitDegrees=5}
 {bone1Name=dyn_ponyTail_21, bone2Name=dyn_ponyTail_02, swing1LimitDegrees=10, swing2LimitDegrees=10, twistLimitDegrees=5}
 {bone1Name=dyn_ponyTail_22, bone2Name=dyn_ponyTail_21, swing1LimitDegrees=15, swing2LimitDegrees=15, twistLimitDegrees=5}
 {bone1Name=dyn_ponyTail_25, bone2Name=dyn_ponyTail_24, swing1LimitDegrees=30, swing2LimitDegrees=30, twistLimitDegrees=5}
 {bone1Name=dyn_ponyTail_26, bone2Name=dyn_ponyTail_25, swing1LimitDegrees=35, swing2LimitDegrees=35, twistLimitDegrees=5}
 {bone1Name=dyn_ponyTail_27, bone2Name=dyn_ponyTail_26, swing1LimitDegrees=40, swing2LimitDegrees=40, twistLimitDegrees=5}
 {bone1Name=dyn_ponyTail_28, bone2Name=dyn_ponyTail_27, swing1LimitDegrees=45, swing2LimitDegrees=45, twistLimitDegrees=5}
 {bone1Name=dyn_ponyTail_29, bone2Name=dyn_ponyTail_28, swing1LimitDegrees=45, swing2LimitDegrees=45, twistLimitDegrees=5}
 {bone1Name=dyn_ponyTail_20, bone2Name=dyn_ponyTail_19, swing1LimitDegrees=45, swing2LimitDegrees=45, twistLimitDegrees=5}
 {bone1Name=dyn_ponyTail_19, bone2Name=dyn_ponyTail_18, swing1LimitDegrees=45, swing2LimitDegrees=45, twistLimitDegrees=5}
 {bone1Name=dyn_ponyTail_18, bone2Name=dyn_ponyTail_17, swing1LimitDegrees=40, swing2LimitDegrees=40, twistLimitDegrees=5}
 {bone1Name=dyn_ponyTail_17, bone2Name=dyn_ponyTail_16, swing1LimitDegrees=35, swing2LimitDegrees=35, twistLimitDegrees=5}
 {bone1Name=dyn_ponyTail_16, bone2Name=dyn_ponyTail_15, swing1LimitDegrees=30, swing2LimitDegrees=30, twistLimitDegrees=5}
 {bone1Name=dyn_ponyTail_15, bone2Name=dyn_ponyTail_14, swing1LimitDegrees=25, swing2LimitDegrees=25, twistLimitDegrees=5}
 {bone1Name=dyn_ponyTail_14, bone2Name=dyn_ponyTail_13, swing1LimitDegrees=20, swing2LimitDegrees=20, twistLimitDegrees=5}
 {bone1Name=dyn_ponyTail_13, bone2Name=dyn_ponyTail_12, swing1LimitDegrees=15, swing2LimitDegrees=15, twistLimitDegrees=5}
 {bone1Name=dyn_ponyTail_12, bone2Name=dyn_ponyTail_02, swing1LimitDegrees=10, swing2LimitDegrees=10, twistLimitDegrees=5}
 {bone1Name=dyn_ponyTail_03, bone2Name=dyn_ponyTail_02, swing1LimitDegrees=10, swing2LimitDegrees=10, twistLimitDegrees=5}
 {bone1Name=dyn_ponyTail_04, bone2Name=dyn_ponyTail_03, swing1LimitDegrees=15, swing2LimitDegrees=15, twistLimitDegrees=5}
 {bone1Name=dyn_ponyTail_05, bone2Name=dyn_ponyTail_04, swing1LimitDegrees=20, swing2LimitDegrees=20, twistLimitDegrees=5}
 {bone1Name=dyn_ponyTail_06, bone2Name=dyn_ponyTail_05, swing1LimitDegrees=25, swing2LimitDegrees=25, twistLimitDegrees=5}
 {bone1Name=dyn_ponyTail_07, bone2Name=dyn_ponyTail_06, swing1LimitDegrees=30, swing2LimitDegrees=30, twistLimitDegrees=5}
 {bone1Name=dyn_ponyTail_08, bone2Name=dyn_ponyTail_07, swing1LimitDegrees=35, swing2LimitDegrees=35, twistLimitDegrees=5}
 {bone1Name=dyn_ponyTail_09, bone2Name=dyn_ponyTail_08, swing1LimitDegrees=40, swing2LimitDegrees=40, twistLimitDegrees=5}
 {bone1Name=dyn_ponyTail_10, bone2Name=dyn_ponyTail_09, swing1LimitDegrees=45, swing2LimitDegrees=45, twistLimitDegrees=5}
 {bone1Name=dyn_ponyTail_11, bone2Name=dyn_ponyTail_10, swing1LimitDegrees=45, swing2LimitDegrees=45, twistLimitDegrees=5}
]
