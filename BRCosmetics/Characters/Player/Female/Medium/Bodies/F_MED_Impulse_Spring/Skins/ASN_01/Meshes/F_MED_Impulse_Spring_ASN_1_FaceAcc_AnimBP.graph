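# Post-copy AnimBP for the Impulse Spring ASN_01 FaceAcc mesh, adding physics-driven ponytail motion to the parent mesh's pose.
# Bool blends choose RigidBody setups for parachute, skydiving, backbling or default; backbling adds alignment constraints, its own physics asset and base-bone scaling.
# Constraints then tie outer strands to dyn_ponyTail_03–11, weighted by crouch, air and melee states.
# EventGraph updates gravity, rigid-alpha and backbling variables.
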AnimBlueprint /BRCosmetics/Characters/Player/Female/Medium/Bodies/F_MED_Impulse_Spring/Skins/ASN_01/Meshes/F_MED_Impulse_Spring_ASN_1_FaceAcc_AnimBP
target_skeleton=/Game/Characters/Player/Female/Medium/Bodies/F_MED_Impulse_Spring/Skins/ASN_01/Meshes/F_MED_Impulse_Spring_ASN_1_FaceAcc_Skeleton
graphs [
 AnimGraph {nodes [
   n1 = "Constraint - dyn_ponyTail_15 (P)"(Alpha=1, ConstraintWeights_0=1) {
    alphaBoolBlend {blendOption=Linear}
    boneToModify {boneName=dyn_ponyTail_15}
    constraintSetup [{targetBone {boneName=spine_05}}]
    constraintWeights [1]
   }
   n2 = "Constraint - dyn_ponyTail_07 (P)"(ComponentPose=@n1, Alpha=1, ConstraintWeights_0=1) {
    alphaBoolBlend {blendOption=Linear}
    boneToModify {boneName=dyn_ponyTail_07}
    constraintSetup [{targetBone {boneName=spine_05}}]
    constraintWeights [1]
   }
   "Constraint - dyn_ponyTail_25 (P)"(ComponentPose=@n2, Alpha=1, ConstraintWeights_0=1) {
    alphaBoolBlend {blendOption=Linear}
    boneToModify {boneName=dyn_ponyTail_25}
    constraintSetup [{targetBone {boneName=spine_05}}]
    constraintWeights [1]
   }
   Copy Pose From Mesh() {bUseAttachedParent=true}
   "Transform (Modify) Bone - Bone: dyn_ponyTail_Base"(Alpha=1, Translation="0,0,0", Rotation="0,0,0", Scale="1,1,1") {scaleMode=BMM_Additive, alphaBoolBlend {blendOption=Linear}, boneToModify {boneName=dyn_ponyTail_Base}}
   n3 = RigidBody() {
    bClampLinearTranslationLimitToRefPose=true
    bOverrideWorldGravity=true
    overridePhysicsAsset=/Game/Characters/Player/Female/Medium/Bodies/F_MED_Impulse_Spring/Skins/ASN_01/Meshes/F_MED_Impulse_Spring_FaceAcc_BackPack_Physics1
    simulationSpace=BaseBoneSpace
    alphaBoolBlend {blendOption=Linear}
    alphaScaleBiasClamp {bClampResult=true}
    baseBoneRef {boneName=spine_05}
    componentAppliedLinearAccClamp=<500 500 250>
    componentLinearAccScale=<0.2 0.2 0.05>
    componentLinearVelScale=<0.2 0.2 0.2>
    overrideWorldGravity {z=-980}
   }
   "Transform (Modify) Bone - Bone: dyn_ponyTail_Base"(ComponentPose=@n3, Alpha=1, Translation="0,0,0", Rotation="0,0,0", Scale="0.7,1,1") {scaleMode=BMM_Additive, alphaBoolBlend {blendOption=Linear}, boneToModify {boneName=dyn_ponyTail_Base}, scale {x=0.7}}
   n4 = "Constraint - dyn_ponyTail_14 (P)"!(ComponentPose=Local To Component(LocalPose=Copy Pose From Mesh() {bCopyCurves=true, bUseAttachedParent=true}), ConstraintWeights_0=1) -> (Pose) {
    alpha=0.1
    alphaBoolBlend {blendOption=Linear}
    boneToModify {boneName=dyn_ponyTail_14}
    constraintSetup [{targetBone {boneName=dyn_backpack_alignment}}]
    constraintWeights [1]
   }
   n5 = "Constraint - dyn_ponyTail_06 (P)"!(ComponentPose=@n4, ConstraintWeights_0=1) -> (Pose) {
    alpha=0.1
    alphaBoolBlend {blendOption=Linear}
    boneToModify {boneName=dyn_ponyTail_06}
    constraintSetup [{targetBone {boneName=dyn_backpack_alignment}}]
    constraintWeights [1]
   }
   n6 = "Constraint - dyn_ponyTail_24 (P)"!(ComponentPose=@n5, ConstraintWeights_0=1) -> (Pose) {
    alpha=0.1
    alphaBoolBlend {blendOption=Linear}
    boneToModify {boneName=dyn_ponyTail_24}
    constraintSetup [{targetBone {boneName=dyn_backpack_alignment}}]
    constraintWeights [1]
   }
   n7 = "Constraint - dyn_ponyTail_15 (P)"!(ComponentPose=@n6, ConstraintWeights_0=1) -> (Pose) {
    alpha=0.1
    alphaBoolBlend {blendOption=Linear}
    boneToModify {boneName=dyn_ponyTail_15}
    constraintSetup [{targetBone {boneName=dyn_backpack_alignment}}]
    constraintWeights [1]
   }
   n8 = "Constraint - dyn_ponyTail_07 (P)"!(ComponentPose=@n7, ConstraintWeights_0=1) -> (Pose) {
    alpha=0.1
    alphaBoolBlend {blendOption=Linear}
    boneToModify {boneName=dyn_ponyTail_07}
    constraintSetup [{targetBone {boneName=dyn_backpack_alignment}}]
    constraintWeights [1]
   }
   n9 = "Constraint - dyn_ponyTail_25 (P)"!(ComponentPose=@n8, ConstraintWeights_0=1) -> (Pose) {
    alpha=0.1
    alphaBoolBlend {blendOption=Linear}
    boneToModify {boneName=dyn_ponyTail_25}
    constraintSetup [{targetBone {boneName=dyn_backpack_alignment}}]
    constraintWeights [1]
   }
   n10 = "Constraint - dyn_ponyTail_14 (P)"(ComponentPose=@n9, ConstraintWeights_0=1) {
    alpha=0.1
    alphaBoolBlend {blendOption=Linear}
    boneToModify {boneName=dyn_ponyTail_14}
    constraintSetup [{targetBone {boneName=dyn_backpack_alignment}}]
    constraintWeights [1]
   }
   n11 = "Constraint - dyn_ponyTail_06 (P)"(ComponentPose=@n10, ConstraintWeights_0=1) {
    alpha=0.1
    alphaBoolBlend {blendOption=Linear}
    boneToModify {boneName=dyn_ponyTail_06}
    constraintSetup [{targetBone {boneName=dyn_backpack_alignment}}]
    constraintWeights [1]
   }
   n12 = "Constraint - dyn_ponyTail_24 (P)"(ComponentPose=@n11, ConstraintWeights_0=1) {
    alpha=0.1
    alphaBoolBlend {blendOption=Linear}
    boneToModify {boneName=dyn_ponyTail_24}
    constraintSetup [{targetBone {boneName=dyn_backpack_alignment}}]
    constraintWeights [1]
   }
   n13 = "Constraint - dyn_ponyTail_15 (P)"(ComponentPose=@n12, ConstraintWeights_0=1) {
    alpha=0.1
    alphaBoolBlend {blendOption=Linear}
    boneToModify {boneName=dyn_ponyTail_15}
    constraintSetup [{targetBone {boneName=dyn_backpack_alignment}}]
    constraintWeights [1]
   }
   n14 = "Constraint - dyn_ponyTail_07 (P)"(ComponentPose=@n13, ConstraintWeights_0=1) {
    alpha=0.1
    alphaBoolBlend {blendOption=Linear}
    boneToModify {boneName=dyn_ponyTail_07}
    constraintSetup [{targetBone {boneName=dyn_backpack_alignment}}]
    constraintWeights [1]
   }
   n15 = "Constraint - dyn_ponyTail_25 (P)"(ComponentPose=@n14, ConstraintWeights_0=1) {
    alpha=0.1
    alphaBoolBlend {blendOption=Linear}
    boneToModify {boneName=dyn_ponyTail_25}
    constraintSetup [{targetBone {boneName=dyn_backpack_alignment}}]
    constraintWeights [1]
   }
   n16 = RigidBody(ComponentPose=@n15) {
    bClampLinearTranslationLimitToRefPose=true
    bOverrideWorldGravity=true
    overridePhysicsAsset=&F_MED_Impulse_Spring_FaceAcc_BackPack_Physics1
    simulationSpace=BaseBoneSpace
    alphaBoolBlend {blendOption=Linear}
    alphaScaleBiasClamp {bClampResult=true}
    baseBoneRef {boneName=spine_05}
    componentAppliedLinearAccClamp=<500 500 250>
    componentLinearAccScale=<0.2 0.2 0.05>
    componentLinearVelScale=<0.2 0.2 0.2>
    overrideWorldGravity {z=-980}
   }
   n17 = "Transform (Modify) Bone - Bone: dyn_ponyTail_Base"(ComponentPose=@n16, Alpha=1, Translation="0,0,0", Rotation="0,0,0", Scale="0.55,1,1") {scaleMode=BMM_Additive, alphaBoolBlend {blendOption=Linear}, boneToModify {boneName=dyn_ponyTail_Base}, scale {x=0.55}}
   n18 = RigidBody(ComponentPose=Local To Component(LocalPose=Copy Pose From Mesh() {bCopyCurves=true, bUseAttachedParent=true})) {
    bClampLinearTranslationLimitToRefPose=true
    bOverrideWorldGravity=true
    overridePhysicsAsset=/Game/Characters/Player/Female/Medium/Bodies/F_MED_Impulse_Spring/Skins/ASN_01/Meshes/F_MED_Impulse_Spring_ASN_1_FaceAcc_hair_Physics
    simulationSpace=BaseBoneSpace
    alphaBoolBlend {blendOption=Linear}
    alphaScaleBiasClamp {bInterpResult=true, interpSpeedDecreasing=5, interpSpeedIncreasing=6}
    baseBoneRef {boneName=spine_05}
    componentAppliedLinearAccClamp=<3000 3000 3000>
    componentLinearAccScale=<0.4 0.6 0.25>
    overrideWorldGravity {z=-980}
   }
   n19 = RigidBody!(
    ComponentPose=Local To Component(LocalPose=Copy Pose From Mesh() {bCopyCurves=true, bUseAttachedParent=true})
    Alpha
    OverrideWorldGravity:Vector="0,0,0"
    ExternalForce
   ) -> (Pose) {
    bClampLinearTranslationLimitToRefPose=true
    bOverrideWorldGravity=true
    overridePhysicsAsset=&F_MED_Impulse_Spring_ASN_1_FaceAcc_hair_Physics
    simulationSpace=BaseBoneSpace
    alphaBoolBlend {blendOption=Linear}
    alphaScaleBiasClamp {bClampResult=true}
    baseBoneRef {boneName=root}
    componentAppliedLinearAccClamp=<500 500 250>
    componentLinearAccScale=<0.05 0.05 0.05>
    componentLinearVelScale=<0.05 0.05 0.05>
   }
   n20 = Blend Poses by bool(
    BlendPose_0=Component To Local(ComponentPose=@n19)
    BlendPose_1=Blend Poses by bool(
     BlendPose_0=Component To Local(ComponentPose=@n17)
     BlendPose_1=Component To Local(ComponentPose=@n18)
     bActiveValue=Get wearingBackbling() -> (wearingBackbling:bool)
    ) {blendType=Linear}
    bActiveValue=Get bIsSkyDiving()
   ) {blendType=Linear}
   n21 = RigidBody!(ComponentPose=Local To Component(LocalPose=Copy Pose From Mesh() {bCopyCurves=true, bUseAttachedParent=true}), Alpha, ExternalForce) -> (Pose) {
    bClampLinearTranslationLimitToRefPose=true
    bOverrideWorldGravity=true
    overridePhysicsAsset=&F_MED_Impulse_Spring_ASN_1_FaceAcc_hair_Physics
    simulationSpace=BaseBoneSpace
    alphaBoolBlend {blendOption=Linear}
    alphaScaleBiasClamp {bClampResult=true}
    baseBoneRef {boneName=root}
    componentAppliedLinearAccClamp=<500 500 250>
    componentLinearAccScale=<0.05 0.05 0.05>
    componentLinearVelScale=<0.05 0.05 0.05>
    overrideWorldGravity {z=-980}
   }
   n22 = Component To Local(ComponentPose=@n21)
   n23 = Get bIsParachuteOpened()
   n24 = Get bIsPlayingMeleeAnim()
   n25 = Get bIsJumping()
   n26 = Get bIsFalling()
   n27 = SelectFloat(A=0.6, B=SelectFloat(A=0.4, B=SelectFloat(A=0.6, B=0.2, bPickA=@n24), bPickA=OR Boolean(A=@n25, B=@n26)), bPickA=Get bIsCrouching())
   n28 = "Constraint - dyn_ponyTail_12 (P)"(
    ComponentPose=Local To Component(LocalPose=Blend Poses by bool(BlendPose_0=@n22, BlendPose_1=@n20, bActiveValue=@n23) {blendType=Linear})
    Alpha=@n27
    ConstraintWeights_0=1
   ) {
    alphaBoolBlend {blendOption=Linear}
    boneToModify {boneName=dyn_ponyTail_12}
    constraintSetup [{targetBone {boneName=dyn_ponyTail_03}}]
    constraintWeights [1]
   }
   n29 = "Constraint - dyn_ponyTail_13 (P)"(ComponentPose=@n28, Alpha=@n27, ConstraintWeights_0=1) {
    alphaBoolBlend {blendOption=Linear}
    boneToModify {boneName=dyn_ponyTail_13}
    constraintSetup [{targetBone {boneName=dyn_ponyTail_04}}]
    constraintWeights [1]
   }
   n30 = "Constraint - dyn_ponyTail_14 (P)"!(ComponentPose=@n29, Alpha:float=@n27, ConstraintWeights_0=1) -> (Pose) {
    alphaBoolBlend {blendOption=Linear}
    boneToModify {boneName=dyn_ponyTail_14}
    constraintSetup [{targetBone {boneName=dyn_ponyTail_05}}]
    constraintWeights [1]
   }
   n31 = "Constraint - dyn_ponyTail_15 (P)"!(ComponentPose=@n30, Alpha:float=@n27, ConstraintWeights_0=1) -> (Pose) {
    alphaBoolBlend {blendOption=Linear}
    boneToModify {boneName=dyn_ponyTail_15}
    constraintSetup [{targetBone {boneName=dyn_ponyTail_06}}]
    constraintWeights [1]
   }
   n32 = "Constraint - dyn_ponyTail_16 (P)"(ComponentPose=@n31, Alpha=@n27, ConstraintWeights_0=1) {
    alphaBoolBlend {blendOption=Linear}
    boneToModify {boneName=dyn_ponyTail_16}
    constraintSetup [{targetBone {boneName=dyn_ponyTail_07}}]
    constraintWeights [1]
   }
   n33 = "Constraint - dyn_ponyTail_17 (P)"(ComponentPose=@n32, Alpha=@n27, ConstraintWeights_0=1) {
    alphaBoolBlend {blendOption=Linear}
    boneToModify {boneName=dyn_ponyTail_17}
    constraintSetup [{targetBone {boneName=dyn_ponyTail_08}}]
    constraintWeights [1]
   }
   n34 = "Constraint - dyn_ponyTail_18 (P)"(ComponentPose=@n33, Alpha=@n27, ConstraintWeights_0=1) {
    alphaBoolBlend {blendOption=Linear}
    boneToModify {boneName=dyn_ponyTail_18}
    constraintSetup [{targetBone {boneName=dyn_ponyTail_09}}]
    constraintWeights [1]
   }
   n35 = "Constraint - dyn_ponyTail_19 (P)"(ComponentPose=@n34, Alpha=@n27, ConstraintWeights_0=1) {
    alphaBoolBlend {blendOption=Linear}
    boneToModify {boneName=dyn_ponyTail_19}
    constraintSetup [{targetBone {boneName=dyn_ponyTail_10}}]
    constraintWeights [1]
   }
   n36 = "Constraint - dyn_ponyTail_20 (P)"(ComponentPose=@n35, Alpha=@n27, ConstraintWeights_0=1) {
    alphaBoolBlend {blendOption=Linear}
    boneToModify {boneName=dyn_ponyTail_20}
    constraintSetup [{targetBone {boneName=dyn_ponyTail_11}}]
    constraintWeights [1]
   }
   n37 = "Constraint - dyn_ponyTail_21 (P)"(ComponentPose=@n36, Alpha=@n27, ConstraintWeights_0=1) {
    alphaBoolBlend {blendOption=Linear}
    boneToModify {boneName=dyn_ponyTail_21}
    constraintSetup [{targetBone {boneName=dyn_ponyTail_03}}]
    constraintWeights [1]
   }
   n38 = "Constraint - dyn_ponyTail_22 (P)"(ComponentPose=@n37, Alpha=@n27, ConstraintWeights_0=1) {
    alphaBoolBlend {blendOption=Linear}
    boneToModify {boneName=dyn_ponyTail_22}
    constraintSetup [{targetBone {boneName=dyn_ponyTail_04}}]
    constraintWeights [1]
   }
   n39 = "Constraint - dyn_ponyTail_23 (P)"(ComponentPose=@n38, Alpha=@n27, ConstraintWeights_0=1) {
    alphaBoolBlend {blendOption=Linear}
    boneToModify {boneName=dyn_ponyTail_23}
    constraintSetup [{targetBone {boneName=dyn_ponyTail_05}}]
    constraintWeights [1]
   }
   n40 = "Constraint - dyn_ponyTail_24 (P)"!(ComponentPose=@n39, Alpha:float=@n27, ConstraintWeights_0=1) -> (Pose) {
    alphaBoolBlend {blendOption=Linear}
    boneToModify {boneName=dyn_ponyTail_24}
    constraintSetup [{targetBone {boneName=dyn_ponyTail_06}}]
    constraintWeights [1]
   }
   n41 = "Constraint - dyn_ponyTail_25 (P)"!(ComponentPose=@n40, Alpha:float=@n27, ConstraintWeights_0=1) -> (Pose) {
    alphaBoolBlend {blendOption=Linear}
    boneToModify {boneName=dyn_ponyTail_25}
    constraintSetup [{targetBone {boneName=dyn_ponyTail_07}}]
    constraintWeights [1]
   }
   n42 = "Constraint - dyn_ponyTail_26 (P)"(ComponentPose=@n41, Alpha=@n27, ConstraintWeights_0=1) {
    alphaBoolBlend {blendOption=Linear}
    boneToModify {boneName=dyn_ponyTail_26}
    constraintSetup [{targetBone {boneName=dyn_ponyTail_08}}]
    constraintWeights [1]
   }
   n43 = "Constraint - dyn_ponyTail_27 (P)"(ComponentPose=@n42, Alpha=@n27, ConstraintWeights_0=1) {
    alphaBoolBlend {blendOption=Linear}
    boneToModify {boneName=dyn_ponyTail_27}
    constraintSetup [{targetBone {boneName=dyn_ponyTail_09}}]
    constraintWeights [1]
   }
   n44 = "Constraint - dyn_ponyTail_28 (P)"(ComponentPose=@n43, Alpha=@n27, ConstraintWeights_0=1) {
    alphaBoolBlend {blendOption=Linear}
    boneToModify {boneName=dyn_ponyTail_28}
    constraintSetup [{targetBone {boneName=dyn_ponyTail_10}}]
    constraintWeights [1]
   }
   n45 = "Constraint - dyn_ponyTail_29 (P)"(ComponentPose=@n44, Alpha=@n27, ConstraintWeights_0=1) {
    alphaBoolBlend {blendOption=Linear}
    boneToModify {boneName=dyn_ponyTail_29}
    constraintSetup [{targetBone {boneName=dyn_ponyTail_11}}]
    constraintWeights [1]
   }
   Output Pose(Result=Component To Local(ComponentPose=@n45)) {layerGroup=None}
  ]}
 EventGraph {nodes [
   Event BlueprintUpdateAnimation()
   > n1 = Sequence!(execute) -> (then_0, then_1, then_2:Exec)
   n2 = Branch(execute=@n1.then_1, Condition=Get bIsSkyDiving())
   > n3 = Branch(Condition=Get bIsParachuteOpened())
   Set gravity_override(execute=[@n3.then, @n2.else], gravity_override="0, 0,-1500")
   > Set rigids alpha(rigids alpha:double=1) -> (Output_Get:double)
   Set gravity_override(execute=@n3.else, gravity_override="0, 0,-1500")
   > Set rigids alpha(rigids alpha=0.5)
   n4 = Gravity Override State Function DEPRECATED(
    execute=@n1.then_0
    CharacterPartAnimInstance=Self-Reference()
    GameTimeInSeconds=GetGameTimeInSeconds()
    GravityOverrideParams=Get Gravity Override Params()
   )
   > Set GravityOverride(GravityOverride=@n4)
   n5 = CheckForBackbling()
   Set wearingBackbling(execute=@n1.then_2, wearingBackbling:bool=@n5) -> (Output_Get:bool)
   n6 = "To Float (Boolean)"(InBool=@n5)
   float - float(B=@n6) {is=title K2Node_CallFunction}
   Set backblingAlpha(backblingAlpha:double=@n6) -> (Output_Get:double)
  ]}
]
variables [
 backblingAlpha {type=number}
 Gravity Override Params {type=GravityOverrideParamsStruct}
 gravity_override {type=Vector}
 GravityOverride {type=Vector}
 rigids alpha {type=number}
 wearingBackbling {type=boolean}
]
